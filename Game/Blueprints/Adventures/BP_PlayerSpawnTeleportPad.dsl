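; Fortnite PvE victory-screen pad that stages players, banners and reward chests at mission end, then handles their departure.
; Mission events clear the area and teleport pawns to mesh sockets, passing player data to clients for banner materials, hidden chests and a camera retried by timer until local pawns exist.
; Client events drive timeline camera and chest animations, plus level-up effects.
; Leaving spawns a drone and starts server-ticked per-player countdowns toward the end-of-mission event.
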(event EventBeginPlay
  ;[1 1 1] Register for mission events
  (MissionHelpers|RegisterforMissionEvents self))

(event EventOnWorldReady
  ;[1 1 1] Clear the surrounding area so there is nothing obstructing our view
  (CallFunction|ClearArea))

(event MissionHelpers|HandleMissionEvent%28v2%29:SetupTeleportOutCamera
  (params (InMissionGuid :type Guid%20Structure) (ObjectiveHandle :type Gameplay%20Tag%20Container%20Structure) (Params :type Fort%20Mission%20Event%20Params%20Object%20Reference) (DO_NOT_USE_THIS_OR_VARIABLES_BELOW :type Boolean) (EventFocus :type Object%20Reference) (EventContent :type Data%20Asset%20Object%20Reference) (EventInstigator :type Actor%20Object%20Reference) (GenericInt :type Integer) (GenericFloat :type Float%20%28single-precision%29) (GenericText :type Text%20%28by%20ref%29) (GameplayTags :type Gameplay%20Tag%20Container%20Structure) (MissionEvent :type Fort%20Mission%20Event%20Structure%20%28by%20ref%29))
  ;[1 1 1] Setup the teleport out camera view
  (Variables|Default|SetbVictoryScreenStarted true)
  (CallFunction|ClearArea)
  (CallFunction|MissionManager|ExecuteMissionEvent(v2) "(CategoryName=\"Native\",EventName=\"PlayTeleportOnToSpawnPad\")"))

(event MissionHelpers|HandleMissionEvent%28v2%29:PlayTeleportOnToSpawnPad
  (params (InMissionGuid :type Guid%20Structure) (ObjectiveHandle :type Gameplay%20Tag%20Container%20Structure) (Params :type Fort%20Mission%20Event%20Params%20Object%20Reference) (DO_NOT_USE_THIS_OR_VARIABLES_BELOW :type Boolean) (EventFocus :type Object%20Reference) (EventContent :type Data%20Asset%20Object%20Reference) (EventInstigator :type Actor%20Object%20Reference) (GenericInt :type Integer) (GenericFloat :type Float%20%28single-precision%29) (GenericText :type Text%20%28by%20ref%29) (GameplayTags :type Gameplay%20Tag%20Container%20Structure) (MissionEvent :type Fort%20Mission%20Event%20Structure%20%28by%20ref%29))
  ;[1 1 1] Teleport the players to the spawn pad
  (CallFunction|TeleportAllPlayers)
  (CallFunction|ClientInitSpawnOut (Variables|Default|GetPlayerData))
  (Utilities|FlowControl|Delay 3) ; play step up to pads anims etc
  (CallFunction|MissionManager|ExecuteMissionEvent(v2) "(CategoryName=\"Native\",EventName=\"FinishedTeleportOnToSpawnPad\")"))

(event MissionHelpers|HandleMissionEvent%28v2%29:PlayerSpawned
  (params (InMissionGuid :type Guid%20Structure) (ObjectiveHandle :type Gameplay%20Tag%20Container%20Structure) (Params :type Fort%20Mission%20Event%20Params%20Object%20Reference) (DO_NOT_USE_THIS_OR_VARIABLES_BELOW :type Boolean) (EventFocus :type Object%20Reference) (EventContent :type Data%20Asset%20Object%20Reference) (EventInstigator :type Actor%20Object%20Reference) (GenericInt :type Integer) (GenericFloat :type Float%20%28single-precision%29) (GenericText :type Text%20%28by%20ref%29) (GameplayTags :type Gameplay%20Tag%20Container%20Structure) (MissionEvent :type Fort%20Mission%20Event%20Structure%20%28by%20ref%29))
  ;[1 1 1] A player has respawned. If we're at the victory screen, ensure they're on the spawn pad and unable to run around.
  (if (Variables|Default|GetbVictoryScreenStarted)
    (Utilities|FlowControl|Delay 0.5) ; TODO: Remove delay and replace with logic for telling\nwhen a player definitely has their pawn
    (CallFunction|TeleportAllPlayers)
    (CallFunction|ClientInitSpawnOut (Variables|Default|GetPlayerData))))

(event ClientEvent|HandleClientEvent:CameraIntro
  (params (EventSource :type Object%20Reference) (EventFocus :type Object%20Reference) (ClientEvent :type Fort%20Client%20Event%20Structure%20%28by%20ref%29))
  ;[1 1 1] Camera Intro Sequence
  (Variables|Default|SetCameraRelativeLocation (Variables|/Script/Engine.SceneComponent|GetRelativeLocation (Variables|Default|GetCameraTransformProxy)))
  (bind (_direction _cameraoffset) (Timeline|TimelineCameraIntro PlayFromStart)
    (:Update
      (Transformation|SetRelativeLocation (Variables|Default|GetCameraTransformProxy) (+ (Variables|Default|GetCameraRelativeLocation) _cameraoffset) false false))))

(event Custom|ClientInitSpawnOut
  (params (NewPlayerData :type Array%20of%20Spawn%20Teleport%20Pad%20Player%20Data%20Structures))
  ;[1 1 1] Client init spawn out
  (Networking|SwitchHasAuthority
    (:Remote
      (Variables|Default|SetPlayerData NewPlayerData)
      (CallFunction|OnReceivedPlayerData)
      (Utilities|SetBannerMaterials
        (:Out
          (CallFunction|SetBannerVisibility true)
          (Camera|InitializeCamera)
          (bind _returnvalue (Utilities|Time|SetTimerbyEvent (EventDispatchers|CreateEvent :!event Custom|ClientSetCameraView) 0.1 true))
          (Variables|Default|SetClientSetCameraView%20TimerHandle _returnvalue)
          (if (not (Variables|Default|GetbSpawnedChests))
            (Variables|Default|SetbSpawnedChests true)
            (Chests|SpawnLootChests)
            (FortPlayerSpawnPad|PrestreamChestTextures)))))))

(event Custom|ClientStartSpawnInCamera)

(event Custom|UpdateBannerMIDS
  (params (BannerMID :type Material%20Instance%20Dynamic%20Object%20Reference) (InstigatingPlayerState :type Fort%20Player%20State%20Object%20Reference))
  (bind _banner (Variables|/Script/FortniteGame.FortPlayerState|GetBanner InstigatingPlayerState))
  (bind (_iconid _colorid) (Utilities|Struct|BreakPlayerBannerInfo _banner))
  ;[1 1 1] Update Banner Material Instance Dynamic references with new icon/color after loading assets
  (Utilities|IsValid (BlueprintContext|GetFortMcpContext)
    (:Is%20Valid
      (Variables|Default|SetBannerIcon (Utilities|String|StringToName _iconid))
      (Variables|Default|SetBannerColor (Utilities|String|StringToName _colorid))
      (bind _returnvalue (Banner|GetBannerLargeIconAssetfromRowName (Variables|Default|GetBannerIcon) "PhysicalBanner"))
      (bind (_loadedasset _midref _bannercolorname) (BannerAsyncLoadAsset|AsyncLoadBannerAsset _returnvalue BannerMID (Variables|Default|GetBannerColor))
        (:BannerAssetLoadComplete
          (bind _astexture_2d (Utilities|Casting|CastToTexture2D _loadedasset)
            (:then
              _)
            (:CastFailed
              _))
          (bind _returnvalue_2 (Banner|GetBannerColorSecondaryColorfromRowName _bannercolorname))
          (bind _returnvalue_3 (Banner|GetBannerColorPrimaryColorfromRowName _bannercolorname))
          (HelperMethods|GenericUpdateMaterial _midref _astexture_2d _returnvalue_3 _returnvalue_2 "/Game/UI/HomeBase/BannerWidget/HB-Banner-Base.HB-Banner-Base"))))))

(event Custom|ClientSetCameraView
  ;[1 1 1] Update Local Player Camera
  (Variables|Default|SetbAllLocalPlayerCameraViewsInitialized true)
  (for _array_element (Fortnite|GetLocalFortPlayerControllers true)
    ;[1 1 1] ONLY set up the player's camera if they have a valid pawn. If the player was dead they may not have a pawn BUT they will receive a pawn soon, and when they do it will hijack the camera, so make sure they already have a pawn.
    (Utilities|IsValid (Game|Player|GetPlayerPawn _array_element)
      (:Is%20Valid
        (Game|Player|SetViewTargetwithBlend _array_element (Variables|Default|GetCamera) 0.0 "VTBlend_Linear" 0.0 false))
      (:Is%20Not%20Valid
        (Variables|Default|SetbAllLocalPlayerCameraViewsInitialized false))))
  (if (Variables|Default|GetbAllLocalPlayerCameraViewsInitialized)
    (Utilities|Time|ClearandInvalidateTimerbyHandle (Variables|Default|GetClientSetCameraView%20TimerHandle))))

(event ClientEvent|HandleClientEvent:PlayerLeavingEvent
  (params (EventSource :type Object%20Reference) (EventFocus :type Object%20Reference) (ClientEvent :type Fort%20Client%20Event%20Structure%20%28by%20ref%29))
  ;[1 1 1] A player is leaving the game
  (Audio|SpawnSoundAttached "/Game/Sounds/Fort_Gadgets/Player_BeamOut/Player_BeamOut_Cue.Player_BeamOut_Cue" (Variables|Debug|GetStaticMeshComponent))
  (Development|PrintText "NSLOCTEXT(\"\", \"F912741C4422A5C58C08919E428A7A35_InText4\", \"PlayerLeaving\")" false)
  (bind _asfort_player_controller (Utilities|Casting|CastToFortPlayerController EventSource)
    (:then
      (Missions|ServerHandleMissionEventStartLeavingZone _asfort_player_controller))
    (:CastFailed
      (Development|PrintText "NSLOCTEXT(\"[4990E8C44B6F5B3744C0B88B044569BE]\", \"BA7338AC465C21504746D59580F9FFE6\", \"BP_PlayerSpawnTeleportPad: PlayerLeavingEvent: FAILED to cast to FortPlayerController.\")" false))))

(event ClientEvent|HandleClientEvent:HeroLevelUpEvent
  (params (EventSource :type Object%20Reference) (EventFocus :type Object%20Reference) (ClientEvent :type Fort%20Client%20Event%20Structure%20%28by%20ref%29))
  ;[1 1 1] A hero has leveled up
  (bind _asfort_player_pawn (Utilities|Casting|CastToFortPlayerPawn EventSource))
  (CallFunction|TriggerLevelUpEffect _asfort_player_pawn))

(event MissionHelpers|HandleMissionEvent%28v2%29:StartLeavingZone
  (params (InMissionGuid :type Guid%20Structure) (ObjectiveHandle :type Gameplay%20Tag%20Container%20Structure) (Params :type Fort%20Mission%20Event%20Params%20Object%20Reference) (DO_NOT_USE_THIS_OR_VARIABLES_BELOW :type Boolean) (EventFocus :type Object%20Reference) (EventContent :type Data%20Asset%20Object%20Reference) (EventInstigator :type Actor%20Object%20Reference) (GenericInt :type Integer) (GenericFloat :type Float%20%28single-precision%29) (GenericText :type Text%20%28by%20ref%29) (GameplayTags :type Gameplay%20Tag%20Container%20Structure) (MissionEvent :type Fort%20Mission%20Event%20Structure%20%28by%20ref%29))
  ;[1 1 1] Start the leave zone event
  (Development|PrintText (Utilities|Text|FormatText "NSLOCTEXT(\"\", \"6C03D82D4F1103779A458AA32D881265_Format0\", \"Server Mission Event: Start Leaving\")") false)
  (bind _asfort_start_leaving_zone_params (Utilities|Casting|CastToFortStartLeavingZoneParams Params))
  (Utilities|IsValid (params|BreakParams _asfort_start_leaving_zone_params :!class /Script/FortniteGame.FortStartLeavingZoneParams)
    (:Is%20Valid
      (Utilities|IsValid (Game|Player|GetPlayerPawn (params|BreakParams _asfort_start_leaving_zone_params :!class /Script/FortniteGame.FortStartLeavingZoneParams))
        (:Is%20Valid
          (CallFunction|StartDroneSpawnOut (Game|Player|GetPlayerPawn (params|BreakParams _asfort_start_leaving_zone_params :!class /Script/FortniteGame.FortStartLeavingZoneParams)))
          (Networking|SwitchHasAuthority
            (:Authority
              (Utilities|Time|SetTimerbyEvent (delegate Custom|LogoutTick) 0.1 true))))
        (:Is%20Not%20Valid
          ;[1 1 1] Fallback case when the player controller doesn't have a valid pawn
          (Networking|SwitchHasAuthority
            (:Authority
              (CallFunction|PlayerLeavingEndedEvent (params|BreakParams _asfort_start_leaving_zone_params :!class /Script/FortniteGame.FortStartLeavingZoneParams)))))))))

(event Custom|LogoutTick
  (bind _playerlogouttimes (Variables|Default|GetPlayerLogoutTimes))
  ;[1 1 1] Tick to deal with  multiple delays issue, not great but fixes it for now
  (Networking|SwitchHasAuthority
    (:Authority
      (for (_array_element _array_index) _playerlogouttimes
        (if (> _array_element 0.0)
          (Utilities|Array|SetArrayElem _playerlogouttimes _array_index (- _array_element 0.1))
          (if (<= (- _array_element 0.1) 0.0)
            (CallFunction|PlayerLeavingEndedEvent (Pawn|GetController (.PlayerPawn_5_A859584247B43636129EE9A40B8991F7 (Utilities|Struct|BreakSpawnTeleportPadPlayerData (Utilities|Array|Get(acopy) (Variables|Default|GetPlayerData) _array_index)))))))))))

(event Custom|PlayerLeavingEndedEvent
  (params (Controller :type Controller%20Object%20Reference))
  ;[1 1 1] Client Player Left Event: Only the local player that started the leaving request should get the end of mission event
  (Development|PrintText (Utilities|Text|FormatText "NSLOCTEXT(\"\", \"44D577294479F4209AAD649C3622413C_Format0\", \"PlayerLeavingEnded\")") false)
  (Networking|SwitchHasAuthority
    (:Remote
      (if (Utilities|Array|ContainsItem (Fortnite|GetLocalFortPlayerControllers true) (Utilities|Casting|CastToFortPlayerController Controller))
        (Development|PrintText "NSLOCTEXT(\"\", \"5FFB3F764FCB39EB01C32E9F1CF9F3A8_InText4\", \"PlayerLeft UI\")")
        (CallFunction|ClientEvent|ExecuteClientEvent "(CategoryName=\"EndOfMission\",EventName=\"PlayerLeftEvent\")")))))

(event ClientEvent|HandleClientEvent:ChestOutro
  (params (EventSource :type Object%20Reference) (EventFocus :type Object%20Reference) (ClientEvent :type Fort%20Client%20Event%20Structure%20%28by%20ref%29))
  (bind (_camerarelativerotation_roll _camerarelativerotation_pitch _camerarelativerotation_yaw) (Math|Rotator|BreakRotator (Variables|GetCameraRelativeRotation_Roll)))
  ;[1 1 1] Chest Outro
  (if (Utilities|Array|IsValidIndex (Variables|FortPlayerSpawnPad|GetChests))
    (Variables|Chests|SetChestPosition (Variables|/Script/Engine.SceneComponent|GetRelativeLocation (Utilities|Array|Get(acopy) (Variables|FortPlayerSpawnPad|GetChests))))
    (bind (_direction _chest_whiteout _chestoffset) (Timeline|TimelineChestOutro PlayFromStart)
      (:Update
        (Rendering|Material|SetScalarParameterValueonMaterials (Utilities|Array|Get(acopy) (Variables|FortPlayerSpawnPad|GetChests)) "whiteOut" _chest_whiteout)
        (Transformation|SetRelativeLocation (Utilities|Array|Get(acopy) (Variables|FortPlayerSpawnPad|GetChests)) (+ (Variables|Chests|GetChestPosition) _chestoffset) false false))
      (:Finished
        (Rendering|SetVisibility (Utilities|Array|Get(acopy) (Variables|FortPlayerSpawnPad|GetChests)) false false)))
    (else
      _))
  (for _array_element (Variables|FortPlayerSpawnPad|GetChests)
    (Components|DestroyComponent _array_element))
  (Utilities|Array|Clear (Variables|FortPlayerSpawnPad|GetChests))
  (Variables|Default|SetCameraRelativeLocation (Variables|/Script/Engine.SceneComponent|GetRelativeLocation (Variables|Default|GetCameraTransformProxy)))
  (Variables|Default|SetCameraRelativeRotation (Variables|/Script/Engine.SceneComponent|GetRelativeRotation (Variables|Default|GetCameraTransformProxy)))
  (bind (_direction_2 _camerapitchoffset _cameraoffset) (Timeline|TimelineChestCameraOutro PlayFromStart)
    (:Update
      (Transformation|SetRelativeLocation (Variables|Default|GetCameraTransformProxy) (+ (Variables|Default|GetCameraRelativeLocation) _cameraoffset) false false)
      (Transformation|SetRelativeRotation (Variables|Default|GetCameraTransformProxy) (Math|Rotator|MakeRotator _camerarelativerotation_roll (+ _camerarelativerotation_pitch _camerapitchoffset) _camerarelativerotation_yaw) false false))
    (:Finished
      (CallFunction|ClientEvent|ExecuteClientEvent "(CategoryName=\"EndOfMission\",EventName=\"ChestOutroDone\")"))))

(event ClientEvent|HandleClientEvent:ChestResultFocus
  (params (EventSource :type Object%20Reference) (EventFocus :type Object%20Reference) (ClientEvent :type Fort%20Client%20Event%20Structure%20%28by%20ref%29))
  ;[1 1 1] Chest Result Focus
  ;;[1 1 1] Hide the remaining chests
  (for (_array_element _array_index) (Variables|FortPlayerSpawnPad|GetChests)
    (if (!= _array_index 0)
      (Rendering|SetVisibility _array_element false false)))
  (Variables|Default|SetCameraRelativeLocation (Variables|/Script/Engine.SceneComponent|GetRelativeLocation (Variables|Default|GetCameraTransformProxy)))
  (bind (_direction _powerswipeanimation _camerapositionoffset) (Timeline|TimelineChestFinalFocus PlayFromStart)
    (:Update
      (Transformation|SetRelativeLocation (Variables|Default|GetCameraTransformProxy) (+ (Variables|Default|GetCameraRelativeLocation) _camerapositionoffset) false false)
      (if (Utilities|Array|IsValidIndex (Variables|FortPlayerSpawnPad|GetChests))
        (Rendering|Material|SetScalarParameterValueonMaterials (Utilities|Array|Get(acopy) (Variables|FortPlayerSpawnPad|GetChests)) "PowerSwipeAnim" _powerswipeanimation)))
    (:Finished
      (CallFunction|ClientEvent|ExecuteClientEvent "(CategoryName=\"EndOfMission\",EventName=\"ChestResultFocusDone\")"))))

(event ClientEvent|HandleClientEvent:ChestAddBadge
  (params (EventSource :type Object%20Reference) (EventFocus :type Object%20Reference) (ClientEvent :type Fort%20Client%20Event%20Structure%20%28by%20ref%29))
  ;[1 1 1] Chest Add Badge
  (if (Utilities|Array|IsValidIndex (Variables|FortPlayerSpawnPad|GetChests))
    (bind (_direction _chest_squash_and_stretch _chest_whiteout) (Timeline|TimelineChestAddBadge PlayFromStart)
      (:Update
        (Rendering|Material|SetScalarParameterValueonMaterials (Utilities|Array|Get(acopy) (Variables|FortPlayerSpawnPad|GetChests)) "Squash and stretch" _chest_squash_and_stretch)
        (Rendering|Material|SetScalarParameterValueonMaterials (Utilities|Array|Get(acopy) (Variables|FortPlayerSpawnPad|GetChests)) "whiteOut" _chest_whiteout)))))

(event ClientEvent|HandleClientEvent:ChestIntro
  (params (EventSource :type Object%20Reference) (EventFocus :type Object%20Reference) (ClientEvent :type Fort%20Client%20Event%20Structure%20%28by%20ref%29))
  (bind _chests (Variables|FortPlayerSpawnPad|GetChests))
  (bind _output (Utilities|Array|Get(acopy) _chests))
  (bind (_camerarelativerotation_roll _camerarelativerotation_pitch _camerarelativerotation_yaw) (Math|Rotator|BreakRotator (Variables|GetCameraRelativeRotation_Roll)))
  ;[1 1 1] Chest Intro
  (Variables|Default|SetCameraRelativeRotation (Variables|/Script/Engine.SceneComponent|GetRelativeRotation (Variables|Default|GetCameraTransformProxy)))
  (bind (_direction _camerapitchoffset) (Timeline|TimelineChestCameraIntro PlayFromStart)
    (:Update
      (Transformation|SetRelativeRotation (Variables|Default|GetCameraTransformProxy) (Math|Rotator|MakeRotator _camerarelativerotation_roll (+ _camerarelativerotation_pitch _camerapitchoffset) _camerarelativerotation_yaw) false false))
    (:Finished
      (Utilities|FlowControl|Sequence
        (:then_0
          ;; Remove all chests that are not the reward from the array
          (while (and (!= (Variables|Chests|GetCurrentChestLevel) (FortPlayerSpawnPad|GetRewardedChestIndex)) (> (Utilities|Array|Length (Variables|FortPlayerSpawnPad|GetChests)) 1))
            (Rendering|SetVisibility (Utilities|Array|Get(acopy) (Variables|FortPlayerSpawnPad|GetChests)) false false)
            (Components|DestroyComponent (Utilities|Array|Get(acopy) (Variables|FortPlayerSpawnPad|GetChests)))
            (Utilities|Array|RemoveIndex (Variables|FortPlayerSpawnPad|GetChests))
            (Math|Integer|IncrementInt (Variables|Chests|GetCurrentChestLevel)))
          ;;;[1 1 1] Hide the remaining chests
          (for (_array_element _array_index) (Variables|FortPlayerSpawnPad|GetChests)
            (if (!= _array_index 0)
              (Rendering|SetVisibility _array_element false false))))
        (:then_1
          ;; Set the reward chest to the position
          (Rendering|SetVisibility _output true false)
          (Timeline|TimelineChestIntroNextChest PlayFromStart
            (:Update
              (Transformation|SetRelativeLocation _output "0, 0, 0" false false))
            (:Finished
              (Audio|PlaySound2D (Variables|Audio|GetChestDisplay))
              (CallFunction|ClientEvent|ExecuteClientEvent "(CategoryName=\"EndOfMission\",EventName=\"ChestIntroDone\")"))))))))

(fn ChestSkipToResultFocusTimelineStops
  (Components|Timeline|Stop (Variables|BP_PlayerSpawnTeleportPAd|GetTimelineChestAddBadge))
  (Utilities|FlowControl|Sequence
    (:then_0
      (if (or (Components|Timeline|IsPlaying (Variables|BP_PlayerSpawnTeleportPAd|GetTimelineChestCameraIntro)) (Components|Timeline|IsPlaying (Variables|BP_PlayerSpawnTeleportPAd|GetTimelineChestIntroNextChest)))
        (Components|Timeline|Stop (Variables|BP_PlayerSpawnTeleportPAd|GetTimelineChestIntroNextChest))
        (CallFunction|ClientEvent|ExecuteClientEvent "(CategoryName=\"EndOfMission\",EventName=\"ChestIntroDone\")")))
    (:then_1
      (if (Components|Timeline|IsPlaying (Variables|BP_PlayerSpawnTeleportPAd|GetTimelineChestFinalFocus))
        (Components|Timeline|Stop (Variables|BP_PlayerSpawnTeleportPAd|GetTimelineChestFinalFocus))
        (CallFunction|ClientEvent|ExecuteClientEvent "(CategoryName=\"EndOfMission\",EventName=\"ChestResultFocusDone\")")))))

(fn ClearArea
  (if (Variables|Default|GetbShouldUseDynamicBuildingDeconstructor)
    (MissionHelpers|CreateDynamicBuildingDeconstructor (Transformation|GetWorldLocation (Variables|Default|GetDestructionVolume)) (Variables|/Script/Engine.BoxComponent|GetBoxExtent (Variables|Default|GetDestructionVolume)) true true)
    (else
      _))
  ;[1 1 1] Destroy all held teleporters on the victory screen.  Otherwise, they will drop and be visible.
  (if (Variables|Default|GetbVictoryScreenStarted)
    (bind _outactors (Actor|GetAllActorsOfClass "/Game/Abilities/Player/Generic/Gadgets/Teleporter/B_Teleporter.B_Teleporter_C"))
    (for _array_element _outactors
      (Utilities|IsValid _array_element
        (:Is%20Valid
          (Variables|/Game/Abilities/Player/Generic/Gadgets/Teleporter/B_Teleporter.B_Teleporter_C|GetHoldingPawn _array_element
            (:then
              (Actor|DestroyActor _array_element))
            (:else)))))))

(fn CreateBannerMID
  (params (PlayerIdx :type Integer))
  (bind _returnvalue (Rendering|Material|CreateDynamicMaterialInstance (Utilities|Select :Index PlayerIdx :Option%200 (Variables|Default|GetS_Spawn_Flag_01) :Option%201 (Variables|Default|GetS_Spawn_Flag_02) :Option%202 (Variables|Default|GetS_Spawn_Flag_03) :Option%203 (Variables|Default|GetS_Spawn_Flag_04)) 1 "/Game/Blueprints/Adventures/SpawnTeleporter/M_SpawnFlag.M_SpawnFlag" "None" :!class /Script/Engine.PrimitiveComponent))
  (bind _structout (Utilities|Struct|SetmembersinSpawnTeleportPadPlayerData (Utilities|Array|Get(acopy) (Variables|Default|GetPlayerData) PlayerIdx) _returnvalue))
  (Utilities|Array|SetArrayElem (Variables|Default|GetPlayerData) PlayerIdx _structout))

(fn EnableLocalPlayerInput
  (params (bEnable :type Boolean :default true))
  (for _array_element (Fortnite|GetLocalFortPlayerControllers true)
    (Game|Cinematic|SetCinematicModewithWeapon _array_element (not bEnable) false (Emote|IsPlayingEmote _array_element) false true true)))

(fn ForceThirdPerson
  (for _array_element (Fortnite|GetLocalFortPlayerControllers true)
    (Camera|SetFirstPersonCamera _array_element false)))

(fn GetDroneTeleportTransform
  (params (PlayerIdx :type Integer))
  (return (Math|Transform|MakeTransform (Math|Vector|MakeVector (.x (.location (CallFunction|GetPlayerTeleportTransform PlayerIdx))) (.y (.location (CallFunction|GetPlayerTeleportTransform PlayerIdx))) (+ (.z (.location (CallFunction|GetPlayerTeleportTransform PlayerIdx))) (Variables|Default|GetDroneHeight))) (Math|Rotator|MakeRotator (.roll (.rotation (CallFunction|GetPlayerTeleportTransform PlayerIdx))) (.pitch (.rotation (CallFunction|GetPlayerTeleportTransform PlayerIdx))) (+ (.yaw (.rotation (CallFunction|GetPlayerTeleportTransform PlayerIdx))) 90)) (.scale (CallFunction|GetPlayerTeleportTransform PlayerIdx)))))

(fn GetPlayerIndex
  (params (PlayerPawn :type Fort%20Player%20Pawn%20Object%20Reference))
  (for (_array_element _array_index) (Variables|Default|GetPlayerData)
    (if (and (Utilities|IsValid (.PlayerPawn_5_A859584247B43636129EE9A40B8991F7 (Utilities|Struct|BreakSpawnTeleportPadPlayerData _array_element)) :!class /Script/Engine.KismetSystemLibrary) (Utilities|Equal(Object) (.PlayerPawn_5_A859584247B43636129EE9A40B8991F7 (Utilities|Struct|BreakSpawnTeleportPadPlayerData _array_element)) PlayerPawn))
      (return _array_index)))
  (return 0))

(fn GetPlayerTeleportTransform
  (params (PlayerIndex :type Integer))
  (return (Transformation|GetSocketTransform (Variables|Debug|GetStaticMeshComponent) (Utilities|String|StringToName (Utilities|String|Append "Player" (Utilities|String|ToString(Integer) (+ PlayerIndex 1)))) "RTS_World")))

(fn InitBannerMIDs
  (for (_array_element _array_index) (Variables|Default|GetPlayerData)
    (Utilities|IsValid (.PlayerPawn_5_A859584247B43636129EE9A40B8991F7 (Utilities|Struct|BreakSpawnTeleportPadPlayerData _array_element))
      (:Is%20Valid
        (CallFunction|CreateBannerMID _array_index)))))

(fn InitializeCamera
  (if (not (Variables|Default|GetbInitializedCamera))
    (Variables|Default|SetbInitializedCamera true)
    (bind _returnvalue (Game|SpawnActorfromClass "/Script/Engine.CameraActor" (Transformation|GetWorldTransform (Variables|Default|GetCameraTransformProxy)) "AlwaysSpawn" "MultiplyWithRoot" 0 0))
    (Variables|Default|SetCamera _returnvalue)
    (Variables|/Script/Engine.CameraComponent|SetbConstrainAspectRatio false (Variables|/Script/Engine.CameraActor|GetCameraComponent (Variables|Default|GetCamera)))
    (Variables|/Script/Engine.CameraComponent|SetFieldOfView 70 (Variables|/Script/Engine.CameraActor|GetCameraComponent (Variables|Default|GetCamera)))
    (Transformation|AttachActorToComponent (Variables|Default|GetCamera) (Variables|Default|GetCameraTransformProxy) "None" "KeepWorld" "KeepWorld" "KeepWorld" false)))

(fn InitPlayerData
  (bind _playerdata (Variables|Default|GetPlayerData))
  (Utilities|Array|Clear _playerdata)
  (Utilities|Array|Resize _playerdata (Variables|Default|GetMaximumNumberOfPlayers)))

(fn OnReceivedPlayerData
  (Camera|ForceThirdPerson)
  (CallFunction|InitBannerMIDs)
  (bind _asfort_game_state (Utilities|Casting|CastToFortGameState (Game|GetGameState))
    (:then
      (CallFunction|EnableLocalPlayerInput (Utilities|Select :Index (GameState|GetGameplayState _asfort_game_state) :NormalGameplay true :WaitingToStart true :EndOfZone false :EnteringZone true :LeavingZone true :Invalid true)))
    (:CastFailed
      (CallFunction|EnableLocalPlayerInput true)))
  (if (FortGlobals|IsNewRewardFlowEnabled)
    ;[1 1 1] New results screen has flags aligned differently to align with the UI elements.
    (Transformation|SetRelativeRotation (Variables|Default|GetS_Spawn_Flag_02) "0,180.000000, 0" false false)
    (Transformation|SetRelativeLocation (Variables|Default|GetS_Spawn_Flag_04) "-260.000000,-207.381683, 0" false false)
    (Transformation|SetRelativeLocation (Variables|Default|GetS_Spawn_Flag_02) "-105.000000,-207.381683, 0" false false)
    (Transformation|SetRelativeLocation (Variables|Default|GetS_Spawn_Flag_01) "105.000000,-207.381683, 0" false false)
    (Transformation|SetRelativeLocation (Variables|Default|GetS_Spawn_Flag_03) "260.000000,-207.381683, 0" false false)))

(fn SetBannerVisibility
  (params (bVisible :type Boolean))
  (if bVisible
    (for (_array_element _array_index) (Variables|Default|GetPlayerData)
      (Utilities|IsValid (.PlayerPawn_5_A859584247B43636129EE9A40B8991F7 (Utilities|Struct|BreakSpawnTeleportPadPlayerData _array_element))
        (:Is%20Valid
          (Rendering|SetVisibility (Utilities|Select :Index _array_index :Option%200 (Variables|Default|GetS_Spawn_Flag_01) :Option%201 (Variables|Default|GetS_Spawn_Flag_02) :Option%202 (Variables|Default|GetS_Spawn_Flag_03) :Option%203 (Variables|Default|GetS_Spawn_Flag_04)) true false))))
    (else
      (Rendering|SetVisibility (Variables|Default|GetS_Spawn_Flag_01) false false))))

(fn SpawnAllDrones
  (for (_array_element _array_index) (Variables|Default|GetPlayerData)
    (Utilities|IsValid (.PlayerPawn_5_A859584247B43636129EE9A40B8991F7 (Utilities|Struct|BreakSpawnTeleportPadPlayerData _array_element))
      (:Is%20Valid
        (CallFunction|SpawnDrone _array_index)))))

(fn SpawnDrone
  (params (PlayerIndex :type Integer))
  (Utilities|IsValid (Utilities|Struct|BreakSpawnTeleportPadPlayerData (Utilities|Array|Get(acopy) (Variables|Default|GetPlayerData) PlayerIndex))
    (:Is%20Valid
      (Rendering|SetActorHiddenInGame false))
    (:Is%20Not%20Valid
      (bind _returnvalue (Game|SpawnActorfromClass "/Game/Blueprints/Adventures/SpawnTeleporter/BP_VictoryDrone.BP_VictoryDrone_C" (CallFunction|GetDroneTeleportTransform PlayerIndex) "AlwaysSpawn" "OverrideRootScale" self false false 0 false false false 0))
      (bind _structout (Utilities|Struct|SetmembersinSpawnTeleportPadPlayerData (Utilities|Array|Get(acopy) (Variables|Default|GetPlayerData) PlayerIndex) _returnvalue))
      (Utilities|Array|SetArrayElem (Variables|Default|GetPlayerData) PlayerIndex _structout))))

(fn SpawnLootChests
  (bind (_returnvalue_x _returnvalue_y) (Math|Vector|BreakVector (Transformation|GetWorldLocation (Variables|Default|GetLootChestCenter))))
  (bind (_nextchestspawntransform_location _nextchestspawntransform_rotation _nextchestspawntransform_scale) (Math|Transform|BreakTransform (Variables|GetNextChestSpawnTransform_Location)))
  ;[1 1 1] Initial chest spawn transform is set here first since it isn't adjusted to the ground.
  (LocalVariables|SetNextChestSpawnTransform (Transformation|GetRelativeTransform (Variables|Default|GetLootChestCenter)))
  (Utilities|FlowControl|Sequence
    (:then_0
      ;[1 1 1] Trace to set the center component's ground position where the chest spawning occurs.
      (bind (_outhit _returnvalue) (Collision|LineTraceByChannel (+ (Transformation|GetWorldLocation (Variables|Default|GetLootChestCenter)) "0,0.000000,50.000000") (+ (Transformation|GetWorldLocation (Variables|Default|GetLootChestCenter)) "0, 0,-50.000000") "TraceTypeQuery10" true))
      (if _returnvalue
        (Transformation|SetWorldLocation (Variables|Default|GetLootChestCenter) (Math|Vector|MakeVector _returnvalue_x _returnvalue_y (.z (.Location (Collision|BreakHitResult _outhit)))) false false)))
    (:then_1
      (for _array_element (Variables|Chests|GetChestMeshes)
        (bind _returnvalue_2 (AddComponent|NODE_AddStaticMeshComponent-0 :bManualAttachment true :RelativeTransform (LocalVariables|GetNextChestSpawnTransform)))
        (Collision|SetCollisionEnabled _returnvalue_2 "NoCollision")
        (Transformation|AttachComponentToComponent _returnvalue_2 (Variables|Default|GetLootChestCenter) "None" "KeepRelative" "KeepRelative" "KeepRelative" true)
        (Rendering|SetVisibility _returnvalue_2 false false)
        (Utilities|Array|Add (Variables|FortPlayerSpawnPad|GetChests) _returnvalue_2)
        (Components|StaticMesh|SetStaticMesh _returnvalue_2 _array_element)
        (LocalVariables|SetNextChestSpawnTransform (Math|Transform|MakeTransform (+ (Variables|Chests|GetChestSpawnOffset) _nextchestspawntransform_location) _nextchestspawntransform_rotation _nextchestspawntransform_scale))))))

(fn StartDroneSpawnOut
  (params (PlayerPawn :type Fort%20Player%20Pawn%20Object%20Reference))
  (Utilities|IsValid PlayerPawn
    (:Is%20Valid
      (for (_array_element _array_index) (Variables|Default|GetPlayerData)
        (if (Utilities|Equal(Object) (.PlayerPawn_5_A859584247B43636129EE9A40B8991F7 (Utilities|Struct|BreakSpawnTeleportPadPlayerData _array_element)) PlayerPawn)
          (CallFunction|SpawnDrone _array_index)
          (Utilities|Array|SetArrayElem (Variables|Default|GetPlayerLogoutTimes) _array_index (Variables|Default|GetExitDelay)))))))

(fn StartSpawnInSequenceForPlayer
  (params (NewPlayerParams :type Fort%20New%20Player%20Params%20Object%20Reference)))

(fn TeleportAllPlayers
  (bind _newpawn (LocalVariables|GetNewPawn))
  (if (FortGlobals|IsNewRewardFlowEnabled)
    (LocalVariables|SetPlayerPawns (FortPlayerSpawnPad|GetPossessedPlayerPawnsbyID))
    (else
      (bind _returnvalue (Fortnite|AuthorityGetFortPossessedPlayerPawns))
      (LocalVariables|SetPlayerPawns _returnvalue)))
  (for (_array_element _array_index) (LocalVariables|GetPlayerPawns)
    (bind _asfort_player_controller_pv_e (Utilities|Casting|CastToFortPlayerControllerPvE (Pawn|GetController _array_element)))
    (Game|Player|PreparePlayerforDisplay _asfort_player_controller_pv_e)
    (bind _asfort_player_pawn (Utilities|Casting|CastToFortPlayerPawn (Pawn|GetControlledPawn _asfort_player_controller_pv_e)))
    (bind _output_get (LocalVariables|SetNewPawn _asfort_player_pawn))
    (CallFunction|TryExitVehicle _output_get) ; FORT-132374
    (Quickbars|ClientSwapQuickBarFocus _asfort_player_controller_pv_e "Primary" -1 true)
    (Transform|Actor|TeleportPlayerPawn _newpawn (Math|Vector|MakeVector (.x (.location (CallFunction|GetPlayerTeleportTransform _array_index))) (.y (.location (CallFunction|GetPlayerTeleportTransform _array_index))) (+ (.z (.location (CallFunction|GetPlayerTeleportTransform _array_index))) (Components|Capsule|GetScaledCapsuleHalfHeight (Variables|/Script/Engine.Character|GetCapsuleComponent _array_element)))) (.rotation (CallFunction|GetPlayerTeleportTransform _array_index)) true)
    (bind _structout (Utilities|Struct|SetmembersinSpawnTeleportPadPlayerData (Utilities|Array|Get(acopy) (Variables|Default|GetPlayerData) _array_index) _newpawn))
    (Utilities|Array|SetArrayElem (Variables|Default|GetPlayerData) _array_index _structout)
    (EventDispatchers|CallOnPostTeleportPlayerVictoryScreen _asfort_player_controller_pv_e)))

(fn TriggerLevelUpEffect
  (params (LevelingUpPawn :type Fort%20Player%20Pawn%20Object%20Reference))
  (bind _outfortplayerpawns (Fortnite|GetFortPlayerPawns))
  (for (_array_element _array_index) _outfortplayerpawns
    (if (Utilities|Equal(Object) _array_element LevelingUpPawn)
      (Components|Activation|Activate (Utilities|Select :Index _array_index :Option%200 (Variables|Default|GetPlayer1LevelUpPS) :Option%201 (Variables|Default|GetPlayer2LevelUpPS) :Option%202 (Variables|Default|GetPlayer3LevelUpPS) :Option%203 (Variables|Default|GetPlayer4LevelUpPS)) false))))

(fn TryExitVehicle
  (params (PlayerPawn :type Fort%20Player%20Pawn%20Object%20Reference))
  ;[1 1 1] Exit vehicle, if in one
  (if (Vehicle|IsinVehicle PlayerPawn)
    (AthenaMovement|ServeronExitVehicle PlayerPawn "(ExitForceBehavior=ForceOnBlocking,bDestroyVehicleWhenForced=True,ExitSelectionData=(SelectionBehavior=Unspecified,Location=(X=0.000000,Y=0.000000,Z=0.000000)))")))

(fn ConstructionScript
  (Rendering|SetVisibility (Variables|Debug|GetStaticMeshComponent) false false)
  (CallFunction|SetBannerVisibility false)
  (Networking|SwitchHasAuthority
    (:Authority
      (CallFunction|InitPlayerData)
      (Utilities|Array|Resize (Variables|Default|GetPlayerLogoutTimes) (Variables|Default|GetMaximumNumberOfPlayers)))))
